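# MetaSound patch that applies a biquad filter to a stereo audio signal, for reuse inside other MetaSound graphs.
# Cutoff frequency is clamped to the 20 Hz to 20 kHz range, then drives two parallel Biquad Filter nodes, one per channel.
# Both channels share the same filter type, bandwidth and gain inputs, so left and right are processed identically.
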MetaSoundPatch /PilgrimCore/MetaSounds/MSP_StereoBiquadFilter_Jam
inputs [
 Bandwidth {type=Float}
 Cutoff Frequency {type=Float}
 Gain {type=Float}
 In L {type=Audio}
 In R {type=Audio}
 Type {type=Enum:BiquadFilterType}
]
interfaces [MetaSound:1.0]
outputs [
 Out L {type=Audio}
 Out R {type=Audio}
]
pages [Default {
  input_defaults {Bandwidth=1, Cutoff Frequency=20000, Gain=0, In L=null, In R=null, Type=0}
  nodes [
   n1 = Clamp.Clamp:Float(In=^Cutoff Frequency) {also {defaults {Max=20000, Min=20}}}
   BiquadFilter_0 = UE.Biquad Filter:Audio(Bandwidth=^Bandwidth, Gain=^Gain, In=^In R, Type=^Type, Cutoff Frequency=@n1.Value)
   BiquadFilter_1 = UE.Biquad Filter:Audio(Bandwidth=^Bandwidth, Gain=^Gain, In=^In L, Type=^Type, Cutoff Frequency=@n1.Value)
  ]
  links [{dst="@Out R", src=BiquadFilter_0.Out}, {dst="@Out L", src=BiquadFilter_1.Out}]
 }]
