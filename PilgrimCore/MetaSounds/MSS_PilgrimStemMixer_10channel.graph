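# Quad MetaSound source mixing Pilgrim song stems from pre-effect buses, with fadeable gains and a rear Fake bus pair.
# DisruptFX envelopes shift drums, bass, guitar or vocals from the dry mix into an LFO-modulated stereo delay; Whammy inputs crossfade bass and guitar into pitch shifters.
# The mix feeds band, amplitude and FFT analysis plus a volume-follow bus.
# A Music Stems Player supplies only MIDI clock and finish trigger here.
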MetaSoundSource /PilgrimCore/MetaSounds/MSS_PilgrimStemMixer_10channel
inputs [
 Bass Gain {type=Float}
 Crossover Frequencies {type=Float:Array, access=Constructor}
 DisruptFXBass {type=Trigger}
 DisruptFXDrums {type=Trigger}
 DisruptFXGuitar {type=Trigger}
 DisruptFXVocals {type=Trigger}
 DisruptStop {type=Trigger}
 Drums Gain {type=Float}
 Fake Gain {type=Float}
 Float {type=Float}
 Guitar Gain {type=Float}
 Seek Target {type=MusicSeekTarget}
 Speed {type=Float}
 Stems Asset {type=StemsAsset, access=Constructor}
 Track Fade Duration {type=Time}
 Trigger Seek {type=Trigger}
 UE.Source.OnPlay {type=Trigger}
 Value {type=Int32}
 Vocals Gain {type=Float}
 WhammyBass {type=Float}
 WhammyGuitar {type=Float}
]
interfaces [UE.OutputFormat.Quad:1.0, UE.Source.OneShot:1.0, UE.Source:1.1]
outputs [
 Amplitude {type=Float}
 Band1 {type=Float}
 Band2 {type=Float}
 Band3 {type=Float}
 Band4 {type=Float}
 FFTResults {type=HarmonixFFTAnalyzerResults}
 Midi Clock {type=MidiClock}
 "UE.OutputFormat.Quad.Audio:0" {type=Audio}
 "UE.OutputFormat.Quad.Audio:1" {type=Audio}
 "UE.OutputFormat.Quad.Audio:4" {type=Audio}
 "UE.OutputFormat.Quad.Audio:5" {type=Audio}
 UE.Source.OneShot.OnFinished {type=Trigger}
]
pages [Default {
  input_defaults {
   Bass Gain=1
   DisruptFXBass=null
   DisruptFXDrums=null
   DisruptFXGuitar=null
   DisruptFXVocals=null
   DisruptStop=null
   Drums Gain=1
   Fake Gain=1
   Float=1
   Guitar Gain=1
   Seek Target=null
   Speed=1
   Stems Asset=null
   Track Fade Duration=0.001
   Trigger Seek=null
   UE.Source.OnPlay=null
   Value=1
   Vocals Gain=1
   WhammyBass=0
   WhammyGuitar=0
   Crossover Frequencies [200,700,1200]
  }
  nodes [
   n1 = UE.Reroute(Value=^Track Fade Duration)
   n2 = UE.Reroute(Value=@n1.Value)
   n3 = UE.Reroute(Value=@n2.Value)
   n4 = UE.Reroute(Value=@n3.Value)
   n5 = UE.InterpTo:Audio(Target=^Vocals Gain, Interp Time=@n4.Value)
   n6 = UE.Reroute(Value=UE.Reroute(Value=UE.Reroute(Value=@n5.Value).Value).Value)
   n7 = UE.Reroute(Value=^DisruptStop)
   n8 = UE.Reroute(Value=@n7.Value)
   n9 = UE.Reroute(Value=@n8.Value)
   n10 = UE.Reroute(Value=^UE.Source.OnPlay)
   n11 = UE.Reroute(Value=@n10.Value)
   n12 = UE.Reroute(Value=@n11.Value)
   TriggerAnyTriggerAny2_2 = "TriggerAny.Trigger Any (2)"(In 1=UE.Reroute(Value=@n9.Value).Value, In 0=UE.Reroute(Value=@n12.Value).Value)
   ADSREnvelopeADSREnvelope_3 = ADSR Envelope.ADSR Envelope:Float(Trigger Attack=^DisruptFXVocals, Trigger Release=@TriggerAnyTriggerAny2_2.Out) {also {defaults {Attack Time=0.25, Sustain Level=1}}}
   n13 = UE.Reroute(Value=@ADSREnvelopeADSREnvelope_3.Out Envelope)
   n14 = UE.InterpTo:Audio(Target=^Bass Gain, Interp Time=@n2.Value)
   n15 = UE.Reroute(Value=UE.Reroute(Value=UE.Reroute(Value=@n14.Value).Value).Value)
   TriggerAnyTriggerAny2_3 = "TriggerAny.Trigger Any (2)"(In 0=@n11.Value, In 1=@n8.Value)
   ADSREnvelopeADSREnvelope_2 = ADSR Envelope.ADSR Envelope:Float(Trigger Attack=^DisruptFXBass, Trigger Release=@TriggerAnyTriggerAny2_3.Out) {also {defaults {Attack Time=0.25, Sustain Level=1}}}
   n16 = UE.Reroute(Value=@ADSREnvelopeADSREnvelope_2.Out Envelope)
   n17 = UE.Reroute(Value=UE.InterpTo:Audio(Target=^Drums Gain, Interp Time=@n1.Value).Value)
   n18 = UE.Reroute(Value=UE.Reroute(Value=UE.Reroute(Value=@n17.Value).Value).Value)
   TriggerAnyTriggerAny2_1 = "TriggerAny.Trigger Any (2)"(In 1=@n7.Value, In 0=@n10.Value)
   ADSREnvelopeADSREnvelope_1 = ADSR Envelope.ADSR Envelope:Float(Trigger Attack=^DisruptFXDrums, Trigger Release=@TriggerAnyTriggerAny2_1.Out) {also {defaults {Attack Time=0.25, Sustain Level=1}}}
   n19 = UE.Reroute(Value=@ADSREnvelopeADSREnvelope_1.Out Envelope)
   n20 = UE.InterpTo:Audio(Target=^Guitar Gain, Interp Time=@n3.Value)
   n21 = UE.Reroute(Value=UE.Reroute(Value=UE.Reroute(Value=@n20.Value).Value).Value)
   TriggerAnyTriggerAny2_0 = "TriggerAny.Trigger Any (2)"(In 1=@n9.Value, In 0=@n12.Value)
   ADSREnvelopeADSREnvelope_0 = ADSR Envelope.ADSR Envelope:Float(Trigger Attack=^DisruptFXGuitar, Trigger Release=@TriggerAnyTriggerAny2_0.Out) {also {defaults {Attack Time=0.25, Sustain Level=1}}}
   n22 = UE.Reroute(Value=@ADSREnvelopeADSREnvelope_0.Out Envelope)
   AudioBusReader2_5 = "UE.Audio Bus Reader (2)"() {also {defaults {Audio Bus=/SparksSongPlayer/MetaSounds/Buses/Pre-Effect/GuitarsStereoBus}}}
   n23 = UE.Reroute(Value=@AudioBusReader2_5.Out 1)
   n24 = VariableAccessor.Time(Variable=InitVariable.Time().Variable)
   n25 = UE.InterpTo:Audio(
    Target=UE.Reroute(Value=MapRange.MapRange:Float(In=^WhammyGuitar) {also {defaults {Out Range B=-2}}}.Out Value).Value
    Interp Time=@n24.Value
   )
   n26 = UE.Reroute(Value=UE.Reroute(Value=UE.Reroute(Value=@n25.Value).Value).Value)
   n27 = VariableAccessor.Float(Variable=InitVariable.Float().Variable)
   n28 = VariableAccessor.Float(Variable=@n27.Variable)
   DelayPitchShift_0 = UE.Delay Pitch Shift:Audio(In=@n23.Value, Pitch Shift=@n26.Value, Delay Length=@n28.Value)
   n29 = Clamp.Clamp:Float(In=UE.Multiply:Float(PrimaryOperand=UE.Subtract:Float(PrimaryOperand=^WhammyGuitar) {also {defaults {AdditionalOperands=0.05}}}.Out) {also {defaults {AdditionalOperands=1000}}}.Out) {also {defaults {Max=1, Min=0}}}
   n30 = UE.Reroute(Value=UE.Reroute(Value=UE.Reroute(Value=@n29.Value).Value).Value)
   CrossfadeTriggerRouteAudio2_3 = "Crossfade.Trigger Route (Audio, 2):Audio"(In 1=@DelayPitchShift_0.Out, Crossfade Value=@n30.Value, In 0=@n23.Value)
   n31 = UE.Reroute(Value=@CrossfadeTriggerRouteAudio2_3.Out)
   AudioBusReader2_4 = "UE.Audio Bus Reader (2)"() {also {defaults {Audio Bus=/SparksSongPlayer/MetaSounds/Buses/Pre-Effect/DrumsStereoBus}}}
   n32 = UE.Reroute(Value=UE.Reroute(Value=@AudioBusReader2_4.Out 0).Value)
   AudioBusReader2_1 = "UE.Audio Bus Reader (2)"() {also {defaults {Audio Bus=/SparksSongPlayer/MetaSounds/Buses/Pre-Effect/VocalsStereoBus}}}
   n33 = UE.Reroute(Value=@AudioBusReader2_1.Out 0)
   n34 = UE.Reroute(Value=@AudioBusReader2_5.Out 0)
   DelayPitchShift_1 = UE.Delay Pitch Shift:Audio(In=@n34.Value, Pitch Shift=@n26.Value, Delay Length=@n28.Value)
   CrossfadeTriggerRouteAudio2_1 = "Crossfade.Trigger Route (Audio, 2):Audio"(In 1=@DelayPitchShift_1.Out, In 0=@n34.Value, Crossfade Value=@n30.Value)
   n35 = UE.Reroute(Value=UE.Reroute(Value=@CrossfadeTriggerRouteAudio2_1.Out).Value)
   n36 = UE.Reroute(Value=UE.Reroute(Value=@AudioBusReader2_4.Out 1).Value)
   AudioBusReader2_3 = "UE.Audio Bus Reader (2)"() {also {defaults {Audio Bus=/SparksSongPlayer/MetaSounds/Buses/Pre-Effect/FakeStereoBus}}}
   AudioBusReader2_2 = "UE.Audio Bus Reader (2)"() {also {defaults {Audio Bus=/SparksSongPlayer/MetaSounds/Buses/Pre-Effect/BassStereoBus}}}
   n37 = UE.Reroute(Value=@AudioBusReader2_2.Out 0)
   n38 = UE.InterpTo:Audio(Target=UE.Reroute(Value=MapRange.MapRange:Float(In=^WhammyBass) {also {defaults {Out Range B=-2}}}.Out Value).Value, Interp Time=@n24.Value)
   n39 = UE.Reroute(Value=UE.Reroute(Value=UE.Reroute(Value=@n38.Value).Value).Value)
   DelayPitchShift_3 = UE.Delay Pitch Shift:Audio(In=@n37.Value, Pitch Shift=@n39.Value, Delay Length=@n27.Value)
   n40 = Clamp.Clamp:Float(In=UE.Multiply:Float(PrimaryOperand=UE.Subtract:Float(PrimaryOperand=^WhammyBass) {also {defaults {AdditionalOperands=0.05}}}.Out) {also {defaults {AdditionalOperands=1000}}}.Out) {also {defaults {Max=1, Min=0}}}
   n41 = UE.Reroute(Value=UE.Reroute(Value=UE.Reroute(Value=@n40.Value).Value).Value)
   CrossfadeTriggerRouteAudio2_0 = "Crossfade.Trigger Route (Audio, 2):Audio"(In 1=@DelayPitchShift_3.Out, Crossfade Value=@n41.Value, In 0=@n37.Value)
   n42 = UE.Reroute(Value=UE.Reroute(Value=@CrossfadeTriggerRouteAudio2_0.Out).Value)
   n43 = UE.Reroute(Value=@AudioBusReader2_1.Out 1)
   n44 = UE.Reroute(Value=@AudioBusReader2_2.Out 1)
   DelayPitchShift_2 = UE.Delay Pitch Shift:Audio(In=@n44.Value, Pitch Shift=@n39.Value, Delay Length=@n27.Value)
   CrossfadeTriggerRouteAudio2_2 = "Crossfade.Trigger Route (Audio, 2):Audio"(In 1=@DelayPitchShift_2.Out, Crossfade Value=@n41.Value, In 0=@n44.Value)
   n45 = UE.Reroute(Value=UE.Reroute(Value=@CrossfadeTriggerRouteAudio2_2.Out).Value)
   AudioMixerAudioMixerStereo5_0 = "AudioMixer.Audio Mixer (Stereo, 5)"(
    Gain 3=UE.Multiply:Float(AdditionalOperands=@n6.Value, PrimaryOperand=UE.Subtract:Float(AdditionalOperands=@n13.Value) {also {defaults {PrimaryOperand=1}}}.Out).Out
    Gain 1=UE.Multiply:Float(AdditionalOperands=@n15.Value, PrimaryOperand=UE.Subtract:Float(AdditionalOperands=@n16.Value) {also {defaults {PrimaryOperand=1}}}.Out).Out
    Gain 0=UE.Multiply:Float(AdditionalOperands=@n18.Value, PrimaryOperand=UE.Subtract:Float(AdditionalOperands=@n19.Value) {also {defaults {PrimaryOperand=1}}}.Out).Out
    Gain 2=UE.Multiply:Float(AdditionalOperands=@n21.Value, PrimaryOperand=UE.Subtract:Float(AdditionalOperands=@n22.Value) {also {defaults {PrimaryOperand=1}}}.Out).Out
    In 2 R=@n31.Value
    In 0 L=@n32.Value
    In 3 L=@n33.Value
    In 2 L=@n35.Value
    In 0 R=@n36.Value
    In 4 R=UE.Reroute(Value=@AudioBusReader2_3.Out 1).Value
    In 4 L=UE.Reroute(Value=@AudioBusReader2_3.Out 0).Value
    In 1 L=@n42.Value
    In 3 R=@n43.Value
    In 1 R=@n45.Value
   )
   AudioMixerAudioMixerStereo4_0 = "AudioMixer.Audio Mixer (Stereo, 4)"(
    Gain 0=UE.Multiply:Float(PrimaryOperand=@n19.Value, AdditionalOperands=@n18.Value).Out
    Gain 1=UE.Multiply:Float(AdditionalOperands=@n15.Value, PrimaryOperand=@n16.Value).Out
    Gain 3=UE.Multiply:Float(AdditionalOperands=@n6.Value, PrimaryOperand=@n13.Value).Out
    Gain 2=UE.Multiply:Float(AdditionalOperands=@n21.Value, PrimaryOperand=@n22.Value).Out
    In 2 R=@n31.Value
    In 0 L=@n32.Value
    In 3 L=@n33.Value
    In 2 L=@n35.Value
    In 0 R=@n36.Value
    In 1 L=@n42.Value
    In 3 R=@n43.Value
    In 1 R=@n45.Value
   )
   n46 = UE.LFO:Audio() {also {defaults {Frequency=0.5, Min Value=0}}}
   StereoDelay_0 = UE.Stereo Delay:Audio(
    In Left=@AudioMixerAudioMixerStereo4_0.Out L
    In Right=@AudioMixerAudioMixerStereo4_0.Out R
    Delay Time=UE.ConversionFloatToTime(In=MapRange.MapRange:Float(In=@n46.Out) {also {defaults {Out Range A=0.01, Out Range B=0.02}}}.Out Value).Out
   ) {also {defaults {Delay Mode=0, Dry Level=0.5, Feedback=0.6, Wet Level=0.5}}}
   AudioMixerAudioMixerStereo2_1 = "AudioMixer.Audio Mixer (Stereo, 2)"(
    In 1 L=@AudioMixerAudioMixerStereo5_0.Out L
    In 1 R=@AudioMixerAudioMixerStereo5_0.Out R
    In 0 R=UE.Reroute(Value=@StereoDelay_0.Out Right).Value
    In 0 L=UE.Reroute(Value=@StereoDelay_0.Out Left).Value
   )
   AudioBusReader2_0 = "UE.Audio Bus Reader (2)"() {also {defaults {Audio Bus=/SparksSongPlayer/MetaSounds/Buses/Pre-Effect/FakeRearStereoBus}}}
   n47 = UE.InterpTo:Audio(Target=^Fake Gain, Interp Time=UE.Reroute(Value=@n4.Value).Value)
   AudioMixerAudioMixerStereo2_2 = "AudioMixer.Audio Mixer (Stereo, 2)"(In 0 L=@AudioBusReader2_0.Out 0, In 0 R=@AudioBusReader2_0.Out 1, Gain 0=UE.Reroute(Value=UE.Reroute(Value=@n47.Value).Value).Value)
   AudioMixerAudioMixerStereo2_0 = "AudioMixer.Audio Mixer (Stereo, 2)"(
    In 0 L=@AudioMixerAudioMixerStereo2_1.Out L
    In 0 R=@AudioMixerAudioMixerStereo2_1.Out R
    In 1 L=@AudioMixerAudioMixerStereo2_2.Out L
    In 1 R=@AudioMixerAudioMixerStereo2_2.Out L
   )
   AudioBusWriter1_0 = "UE.Audio Bus Writer (1)"(In 0=UE.Add:Audio by Float(PrimaryOperand=@AudioMixerAudioMixerStereo2_0.Out R) {also {defaults {AdditionalOperands=0.3}, version [1,1]}}.Out) {also {defaults {Audio Bus=/SparksCore/Audio/Modulation/AB_Sparks_SongVolume_Follow}}}
   n48 = HarmonixNodes.MultibandAnalyzer(
    Crossover Frequencies=^Crossover Frequencies
    Audio Mono=UE.Add:Audio(PrimaryOperand=@AudioMixerAudioMixerStereo2_0.Out L, AdditionalOperands=@AudioMixerAudioMixerStereo2_0.Out R).Out
   ) {also {defaults {Release Time=0.1}, version [0,1]}}
   n49 = UE.TriggerRepeat(Start=^UE.Source.OnPlay) {also {defaults {Period=0.0333}}}
   Divide_0 = UE.Divide:Float(PrimaryOperand=Array.Get:Float:Array(Array=@n48.Band Levels, Trigger=@n49.RepeatOut) {also {defaults {Index=3}}}.Element) {also {defaults {AdditionalOperands=2}}}
   Divide_1 = UE.Divide:Float(PrimaryOperand=Array.Get:Float:Array(Array=@n48.Band Levels, Trigger=@n49.RepeatOut) {also {defaults {Index=2}}}.Element) {also {defaults {AdditionalOperands=2}}}
   Divide_2 = UE.Divide:Float(PrimaryOperand=Array.Get:Float:Array(Array=@n48.Band Levels, Trigger=@n49.RepeatOut).Element) {also {defaults {AdditionalOperands=2}}}
   Divide_3 = UE.Divide:Float(PrimaryOperand=HarmonixNodes.Peak(Audio Mono=UE.Add:Audio(PrimaryOperand=@AudioMixerAudioMixerStereo2_0.Out L, AdditionalOperands=@AudioMixerAudioMixerStereo2_0.Out R).Out) {also {version [0,1]}}.Peak) {also {defaults {AdditionalOperands=2}}}
   Divide_4 = UE.Divide:Float(PrimaryOperand=Array.Get:Float:Array(Array=@n48.Band Levels, Trigger=@n49.RepeatOut) {also {defaults {Index=1}}}.Element) {also {defaults {AdditionalOperands=2}}}
   HarmonixNodesFFTAnalyzer_0 = HarmonixNodes.FFT Analyzer(In=@AudioMixerAudioMixerStereo2_0.Out L) {also {defaults {Fall Ms=800, Mel Scale Binning="<MetaSoundBuilderResult.SUCCEEDED: 0>", Num Result Bins=64}, version [0,1]}}
   HarmonixNodesMusicStemsPlayer_0 = HarmonixNodes.Music Stems Player(
    Speed=^Speed
    Stems Asset=^Stems Asset
    Transport=HarmonixNodes.TriggerToTransport(Seek Target=^Seek Target, Trigger Seek=^Trigger Seek, Play=^UE.Source.OnPlay).Transport
   )
   &/MSP_StereoBiquadFilter_Jam(
    In L=@AudioMixerAudioMixerStereo4_0.Out L
    In R=@AudioMixerAudioMixerStereo4_0.Out R
    Bandwidth=MapRange.MapRange:Float(In=@n46.Out) {also {defaults {Out Range A=0.75, Out Range B=0.35}}}.Out Value
    Cutoff Frequency=MapRange.MapRange:Float(In=@n46.Out) {also {defaults {Out Range A=1500, Out Range B=7000}}}.Out Value
   ) {also {defaults {Gain=1}}}
   TriggerOnValueChange_0 = UE.Trigger On Value Change:Float()
   /FortMetasound/Patches/PrintFloat(Trigger=@TriggerOnValueChange_0.Trigger) {also {defaults {Label=Whammy}}}
  ]
  links [
   {dst="@UE.OutputFormat.Quad.Audio:0", src=AudioMixerAudioMixerStereo2_1.Out L}
   {dst="@UE.OutputFormat.Quad.Audio:1", src=AudioMixerAudioMixerStereo2_1.Out R}
   {dst="@UE.OutputFormat.Quad.Audio:4", src=AudioMixerAudioMixerStereo2_2.Out L}
   {dst="@UE.OutputFormat.Quad.Audio:5", src=AudioMixerAudioMixerStereo2_2.Out R}
   {dst="@Band4", src=Divide_0.Out}
   {dst="@Band3", src=Divide_1.Out}
   {dst="@Band1", src=Divide_2.Out}
   {dst="@Amplitude", src=Divide_3.Out}
   {dst="@Band2", src=Divide_4.Out}
   {dst="@FFTResults", src=HarmonixNodesFFTAnalyzer_0.Results}
   {dst="@Midi Clock", src=HarmonixNodesMusicStemsPlayer_0.MIDI Clock}
   {dst="@UE.Source.OneShot.OnFinished", src=HarmonixNodesMusicStemsPlayer_0.On Midi Length Reached}
  ]
 }]
properties {
 assetImportData=&self:AssetImportData
 numChannels=4
 outputFormat=Quad
 sampleRate=48000
 soundSubmixObject=/Game/Sounds/SubMixes/Fortnite_Main
 virtualizationMode=PlayWhenSilent
 blockRateOverride {default=100}
 modulationSettings {volumeRouting=Union, volumeModulationDestination {bEnableModulation=true, modulators [/SparksAudioBase/Modulation/CB_Slider_Pilgrim_Music]}}
 soundSubmixSends [{}]
}
