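# One-shot stereo MetaSound source that plays fail music for the results, currently the DimeBlanket Disqualify cue.
# On play, a random-get picks from the Wavs array, whose three entries are all the same cue, and feeds a non-looping stereo wave player.
# Playback end fires OnFinished; the float Input is unconnected.
# Concurrency override limits it to one instance per owner, under the in-game SFX sound class.
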MetaSoundSource /PilgrimCore/MetaSounds/MSS_Results_Music_Randomizer_Fail
inputs [
 Input {type=Float}
 UE.Source.OnPlay {type=Trigger}
 Wavs {type=WaveAsset:Array}
]
interfaces [UE.OutputFormat.Stereo:1.0, UE.Source.OneShot:1.0, UE.Source:1.1]
outputs [
 "UE.OutputFormat.Stereo.Audio:0" {type=Audio}
 "UE.OutputFormat.Stereo.Audio:1" {type=Audio}
 UE.Source.OneShot.OnFinished {type=Trigger}
]
pages [Default {
  input_defaults {
   Input=0
   UE.Source.OnPlay=null
   Wavs [
    /PilgrimCore/MetaSounds/Wavs/BattleStage_Music_Cues/DimeBlanket/FN_BS_DimeBlanket_Disqualify
    &FN_BS_DimeBlanket_Disqualify
    &FN_BS_DimeBlanket_Disqualify
   ]
  }
  nodes [
   ArrayRandomGet_0 = Array.Random Get:WaveAsset:Array(Next=^UE.Source.OnPlay, In Array=^Wavs)
   WavePlayer_0 = UE.Wave Player:Stereo(Play=@ArrayRandomGet_0.On Next, Wave Asset=@ArrayRandomGet_0.Value) {also {defaults {Loop="<MetaSoundBuilderResult.SUCCEEDED: 0>"}}}
  ]
  links [
   {dst="@UE.Source.OneShot.OnFinished", src=WavePlayer_0.On Finished}
   {dst="@UE.OutputFormat.Stereo.Audio:0", src=WavePlayer_0.Out Left}
   {dst="@UE.OutputFormat.Stereo.Audio:1", src=WavePlayer_0.Out Right}
  ]
 }]
properties {
 assetImportData=&self:AssetImportData
 bOverrideConcurrency=true
 numChannels=2
 outputFormat=Stereo
 sampleRate=48000
 soundClassObject=&Fort_SFX_InGame
 volume=0.8
 concurrencyOverrides {bLimitToOwner=true, maxCount=1}
}
